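# Translucent deferred decal material for a tentacle shadow effect.
# It multiplies the RGB of the T_TentacleShadow texture by an OpacityStrength scalar to scale the shadow mask.
# A ShadowColor vector parameter, white by default, is exposed alongside it for tinting.
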Material /LimeRock/Assets/FX/TentacleShadow/M_TentacleShadow_Decal
graph {
 properties {bAutomaticallySetUsageInEditor=false, blendMode=BLEND_Translucent, materialDomain=MD_DeferredDecal, usedShadingModels=Default Lit}
 nodes [
  Multiply(
   A=TextureSample() {texture=&/T_TentacleShadow}.RGB
   B=ScalarParameter() {defaultValue=1, parameterName=OpacityStrength}
  )
  VectorParameter() {parameterName=ShadowColor, defaultValue=<rgba 1 1 1 1>}
 ]
}
