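# M_TentacleShadow_Plane is a masked, unlit material that casts a tentacle-shaped shadow, likely from a plane mesh.
# A Shadow Pass Switch returns a zero constant in normal passes, which should keep the surface itself invisible if it drives the opacity mask.
# In the shadow pass, the TentacleShadow2 texture feeds DitherTemporalAA as the alpha threshold, so dithering approximates soft shadow edges.
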Material /LimeRock/Assets/FX/TentacleShadow/M_TentacleShadow_Plane
graph {
 properties {bAutomaticallySetUsageInEditor=false, blendMode=BLEND_Masked, shadingModel=MSM_Unlit, usedShadingModels=Unlit}
 nodes [ShadowReplace(
   Default=Constant()
   Shadow=MaterialFunctionCall(Alpha Threshold=TextureSample() {texture=&/TentacleShadow2}.RGB) {materialFunction=&DitherTemporalAA}.Result
  )]
}
